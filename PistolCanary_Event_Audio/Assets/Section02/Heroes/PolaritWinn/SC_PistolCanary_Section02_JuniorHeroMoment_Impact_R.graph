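# Sound cue for the right-side (R) impact of the Junior Hero Moment in Section 02 of the PistolCanary event, filed under the PolaritWinn hero.
# A single wave player feeds the output directly, with no randomization, mixing or modulation; the file names no wave asset.
# The cue uses the Fort_Events sound class, a slight volume boost and the shared large point-of-interest attenuation for long-range audibility.
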SoundCue /PistolCanary_Event_Audio/Assets/Section02/Heroes/PolaritWinn/SC_PistolCanary_Section02_JuniorHeroMoment_Impact_R
root=WavePlayer_4
nodes [{id=WavePlayer_4}]
properties {
 attenuationSettings=/PistolCanary_Audio/Assets/Settings/ATT_PistolCanary_POI_Large
 maxDistance=40000
 soundClassObject=&Fort_Events
 volumeMultiplier=1.2
}
